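# Sound cue for generic bullet ammo landing, a loot foley effect in the Loot Collect sound class.
# A Random node picks one of four equally weighted landing waves, feeding a Modulator that slightly varies pitch to reduce repetition.
# It plays at reduced volume with weapon-drop attenuation and a weapon-drop concurrency set.
# The Random node is excluded from branch culling.
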SoundCue /Game/Sounds/Foley_Loot/Ingredients/Munitions/Bullet_Generic_Land_Cue
root=Modulator_1
nodes [
 {id=Modulator_1, children [Random_0], properties {pitchMax=1.1, pitchMin=0.9}}
 {
  id=Random_0
  children [WavePlayer_0,WavePlayer_1,WavePlayer_2,WavePlayer_3]
  properties {bShouldExcludeFromBranchCulling=true, weights [1,1,1,1]}
 }
 {id=WavePlayer_0, properties {soundWaveAssetPtr=&/Bullet_Generic_Ammo_Land_04}}
 {id=WavePlayer_1, properties {soundWaveAssetPtr=&/Bullet_Generic_Ammo_Land_01}}
 {id=WavePlayer_2, properties {soundWaveAssetPtr=&/Bullet_Generic_Ammo_Land_02}}
 {id=WavePlayer_3, properties {soundWaveAssetPtr=&/Bullet_Generic_Ammo_Land_03}}
]
properties {
 attenuationSettings=&Fort_Weapon_Drop_Att
 maxDistance=3000
 soundClassObject=&Fort_Loot_Collect
 volumeMultiplier=0.7
 concurrencyOverrides {resolutionRule=PreventNew}
 concurrencySet [&Weapon_Drop_StopOldest]
}
