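# Sound cue for collecting heavy bullet ammo, an ingredient loot foley sound.
# A mixer layers the heavy ammo collect wave with a quieter generic weighted impact wave, and its middle input is empty.
# The cue plays at low volume through the loot collect sound class and prop collect attenuation.
# A concurrency override caps it at two simultaneous instances, stopping the oldest.
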SoundCue /Game/Sounds/Foley_Loot/Ingredients/Munitions/Bullet_Heavy_Collect_Cue
root=Mixer_1
nodes [
 {id=Mixer_1, children [WavePlayer_6,null,WavePlayer_5], properties {inputVolume [1.2,0.5,0.7]}}
 {id=WavePlayer_5, properties {soundWaveAssetPtr=&/Loot_Foley_Weighted_Impact_01}}
 {id=WavePlayer_6, properties {soundWaveAssetPtr=&/Bullet_Heavy_Ammo_Collect_01}}
]
properties {
 attenuationSettings=/Game/Sounds/Attenuation/Fort_Prop_Collect_Att
 bOverrideConcurrency=true
 maxDistance=2097152
 soundClassObject=&Fort_Loot_Collect
 volumeMultiplier=0.17
 concurrencyOverrides {maxCount=2, resolutionRule=StopOldest}
}
