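# Sound cue for collecting light bullet ammo, in the loot Foley set for munitions ingredients.
# A Mixer layers a light ammo collect wave with a quieter weighted impact Foley wave, at low overall volume.
# It uses prop-collect attenuation and the loot-collect sound class, with a concurrency override that caps simultaneous plays at two by stopping the oldest.
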SoundCue /Game/Sounds/Foley_Loot/Ingredients/Munitions/Bullet_Light_Collect_Cue
root=Mixer_5
nodes [
 {id=Mixer_5, children [WavePlayer_21, WavePlayer_17], properties {inputVolume [1,0.7]}}
 {id=WavePlayer_17, properties {soundWaveAssetPtr=&/Loot_Foley_Weighted_Impact_01}}
 {id=WavePlayer_21, properties {soundWaveAssetPtr=&/Bullet_Light_Ammo_Collect_01}}
]
properties {
 attenuationSettings=/Game/Sounds/Attenuation/Fort_Prop_Collect_Att
 bOverrideConcurrency=true
 maxDistance=2097152
 soundClassObject=&Fort_Loot_Collect
 volumeMultiplier=0.17
 concurrencyOverrides {maxCount=2, resolutionRule=StopOldest}
}
