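# Sound cue for collecting medium bullet ammo, routed to the loot collect sound class with prop-collect attenuation.
# A mixer layers a randomly chosen medium ammo collect wave (four variants) with varied pitch, a weighted impact foley, and a pitch-adjusted light ammo collect wave.
# The medium layer is mixed loudest, and one mixer input is empty.
# A concurrency override caps playback at two instances, stopping the oldest.
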SoundCue /Game/Sounds/Foley_Loot/Ingredients/Munitions/Bullet_Medium_Collect_Cue
root=Mixer_2
nodes [
 {id=Mixer_2, children [Modulator_1,null,WavePlayer_8,Modulator_0], properties {inputVolume [2.5,0.5,0.7,0.7]}}
 {id=Modulator_0, children [WavePlayer_4], properties {pitchMax=0.95, volumeMax=1, volumeMin=1}}
 {id=Modulator_1, children [Random_0], properties {pitchMax=1.1, volumeMax=1, volumeMin=1}}
 {
  id=Random_0
  children [WavePlayer_0,WavePlayer_1,WavePlayer_2,WavePlayer_3]
  properties {bShouldExcludeFromBranchCulling=true, weights [1,1,1,1]}
 }
 {id=WavePlayer_0, properties {soundWaveAssetPtr=&/Bullet_Medium_Ammo_Collect_01}}
 {id=WavePlayer_1, properties {soundWaveAssetPtr=&/Bullet_Medium_Ammo_Collect_02}}
 {id=WavePlayer_2, properties {soundWaveAssetPtr=&/Bullet_Medium_Ammo_Collect_03}}
 {id=WavePlayer_3, properties {soundWaveAssetPtr=&/Bullet_Medium_Ammo_Collect_04}}
 {id=WavePlayer_4, properties {soundWaveAssetPtr=&/Bullet_Light_Ammo_Collect_01}}
 {id=WavePlayer_8, properties {soundWaveAssetPtr=&/Loot_Foley_Weighted_Impact_01}}
]
properties {
 attenuationSettings=/Game/Sounds/Attenuation/Fort_Prop_Collect_Att
 bOverrideConcurrency=true
 maxDistance=2097152
 soundClassObject=&Fort_Loot_Collect
 volumeMultiplier=0.17
 concurrencyOverrides {maxCount=2, resolutionRule=StopOldest}
}
