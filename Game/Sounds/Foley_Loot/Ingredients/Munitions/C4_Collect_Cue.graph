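# Sound cue for collecting C4, a munitions ingredient, routed through the loot collect sound class with prop-collect attenuation.
# A mixer layers a C4 collect sound, a weighted impact foley and a C4 timer warning beep at separate input levels.
# An Enveloper shapes only the beep, raising its pitch while fading its volume out.
# A concurrency override caps simultaneous instances at two, stopping the oldest.
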SoundCue /Game/Sounds/Foley_Loot/Ingredients/Munitions/C4_Collect_Cue
root=Mixer_11
nodes [
 {
  id=Enveloper_0
  children [WavePlayer_52]
  properties {
   pitchCurve {editorCurveData {keys [{time=0}, <at 1.9644 1.3022>]}}
   volumeCurve {editorCurveData {keys [<at -0.0047 1.4779>, {arriveTangent=-0.8844, interpMode=RCIM_Cubic, leaveTangent=-0.8844, time=0.9846, value=1.3061}, <at 1.6753 -0.008>]}}
  }
 }
 {id=Mixer_11, children [WavePlayer_68, null, WavePlayer_46, Enveloper_0], properties {inputVolume [1.2,1,0.8,2.5]}}
 {id=WavePlayer_46, properties {soundWaveAssetPtr=&/Loot_Foley_Weighted_Impact_01}}
 {id=WavePlayer_52, properties {soundWaveAssetPtr=/Game/Sounds/Fort_GamePlay_Sounds/Explosives/timer_beep_C4_warn}}
 {id=WavePlayer_68, properties {soundWaveAssetPtr=&/C4_Collect_Real}}
]
properties {
 attenuationSettings=/Game/Sounds/Attenuation/Fort_Prop_Collect_Att
 bOverrideConcurrency=true
 maxDistance=2097152
 soundClassObject=&Fort_Loot_Collect
 volumeMultiplier=0.259
 concurrencyOverrides {maxCount=2, resolutionRule=StopOldest}
}
