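# Sound cue for collecting gunpowder, a munitions ingredient in the loot foley set.
# A single wave player feeds GunPowder_Collect_Real into a Mixer root whose second input is empty, so no layering occurs.
# Playback runs at reduced volume through the loot-collect sound class and prop-collect attenuation, with a concurrency override that stops the oldest instance beyond a small limit.
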SoundCue /Game/Sounds/Foley_Loot/Ingredients/Munitions/GunPowder_Collect_Cue
root=Mixer_32
nodes [
 {id=Mixer_32, children [WavePlayer_66, null], properties {inputVolume [1,1]}}
 {id=WavePlayer_66, properties {soundWaveAssetPtr=&/GunPowder_Collect_Real}}
]
properties {
 attenuationSettings=/Game/Sounds/Attenuation/Fort_Prop_Collect_Att
 bOverrideConcurrency=true
 maxDistance=2097152
 soundClassObject=&Fort_Loot_Collect
 volumeMultiplier=0.251
 concurrencyOverrides {maxCount=2, resolutionRule=StopOldest}
}
